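# Post-copy animation Blueprint for the female CraneLaugh cosmetic body, adding secondary motion to clothing and a ring.
# It copies the attached parent's pose, chains AnimDynamics on jacket, sleeve and pants bones, then finishes with the outfit's Control Rig.
# A weighted multi-bone constraint anchors the ring to spine, neck and jaw before its simulation.
# Unconnected feather, firework and alternate-tuning nodes remain; the event graph starts setup when character customization loads.
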AnimBlueprint /BRCosmetics/Characters/Player/Female/Medium/Bodies/F_MED_CraneLaugh/Meshes/F_MED_CraneLaugh_AnimBP
target_skeleton=&/F_MED_CraneLaugh_Skeleton
graphs [
 AnimGraph {nodes [
   Anim Dynamics - Bone: dyn_jacket_01(Alpha=0.2, ExternalForce="0,0,0") {
    angularSpringConstant=80
    bAngularSpring=true
    bGravityOverrideInSimSpace=true
    bLinearSpring=true
    lODThreshold=1
    linearSpringConstant=80
    simulationSpace=Actor
    alphaBoolBlend {blendOption=Linear}
    boundBone {boneName=dyn_jacket_01}
    physicsBodyDefinitions [{
      sphereCollisionRadius=0
      boundBone {boneName=dyn_jacket_01}
      boxExtents=<30 30 30>
      constraintSetup {
       angularTargetAxis=AxisY
       angularLimitsMax=<0 30 30>
       angularLimitsMin=<0 -30 -30>
       angularTarget=<0 0 0>
       linearAxesMax=<0 5 5>
       linearAxesMin=<-5 -5 -5>
      }
      localJointOffset=<0 -0.5 0>
     }]
   }
   n1 = Anim Dynamics - Bone: dyn_feathers_01(Alpha=0.15, ExternalForce="0,0,0") {
    angularSpringConstant=80
    bAngularSpring=true
    bGravityOverrideInSimSpace=true
    bLinearSpring=true
    lODThreshold=1
    linearSpringConstant=80
    simulationSpace=Actor
    alphaBoolBlend {blendOption=Linear}
    boundBone {boneName=dyn_feathers_01}
    physicsBodyDefinitions [{
      sphereCollisionRadius=0
      boundBone {boneName=dyn_feathers_01}
      boxExtents=<30 30 30>
      constraintSetup {
       angularTargetAxis=AxisY
       angularLimitsMax=<0 30 30>
       angularLimitsMin=<0 -30 -30>
       angularTarget=<0 0 0>
       linearAxesMax=<10 20 5>
       linearAxesMin=<0 0 -20>
      }
      localJointOffset=<0 0 0>
     }]
   }
   n2 = Anim Dynamics - Bone: dyn_feathers_02(ComponentPose=@n1, Alpha=0.2, ExternalForce="0,0,0") {
    angularSpringConstant=80
    bAngularSpring=true
    bGravityOverrideInSimSpace=true
    bLinearSpring=true
    lODThreshold=1
    linearSpringConstant=80
    simulationSpace=Actor
    alphaBoolBlend {blendOption=Linear}
    boundBone {boneName=dyn_feathers_02}
    physicsBodyDefinitions [{
      sphereCollisionRadius=0
      boundBone {boneName=dyn_feathers_02}
      boxExtents=<30 30 30>
      constraintSetup {
       angularTargetAxis=AxisY
       angularLimitsMax=<0 30 30>
       angularLimitsMin=<0 -30 -30>
       angularTarget=<0 0 0>
       linearAxesMax=<10 20 20>
       linearAxesMin=<0 -5 -5>
      }
      localJointOffset=<0 -0.5 0>
     }]
   }
   Anim Dynamics - Bone: dyn_feathers_03(ComponentPose=@n2, Alpha=0.25, ExternalForce="0,0,0") {
    angularSpringConstant=80
    bAngularSpring=true
    bGravityOverrideInSimSpace=true
    bLinearSpring=true
    lODThreshold=1
    linearSpringConstant=80
    simulationSpace=Actor
    alphaBoolBlend {blendOption=Linear}
    boundBone {boneName=dyn_feathers_03}
    physicsBodyDefinitions [{
      sphereCollisionRadius=0
      boundBone {boneName=dyn_feathers_03}
      boxExtents=<30 30 30>
      constraintSetup {
       angularTargetAxis=AxisY
       angularLimitsMax=<0 30 30>
       angularLimitsMin=<0 -30 -30>
       angularTarget=<0 0 0>
       linearAxesMax=<10 20 20>
       linearAxesMin=<0 -20 -20>
      }
      localJointOffset=<0 0 0>
     }]
   }
   n3 = Anim Dynamics - Bone: dyn_jacket_sleev_01_l(Alpha=0.2, ExternalForce="0,0,0") {
    angularSpringConstant=200
    bAngularSpring=true
    bGravityOverrideInSimSpace=true
    bLinearSpring=true
    bUseGravityOverride=true
    lODThreshold=1
    linearSpringConstant=50
    simulationSpace=RootRelative
    alphaBoolBlend {blendOption=Linear}
    boundBone {boneName=dyn_jacket_sleev_01_l}
    gravityOverride {z=-980}
    physicsBodyDefinitions [{
      sphereCollisionRadius=0
      boundBone {boneName=dyn_jacket_sleev_01_l}
      boxExtents=<20 20 20>
      constraintSetup {
       angularTargetAxis=AxisY
       angularLimitsMax=<15 15 15>
       angularLimitsMin=<-15 -15 -15>
       angularTarget=<0 0 0>
       linearAxesMax=<0 3 1>
       linearAxesMin=<-10 0 -1>
      }
      localJointOffset=<-1 -1 -1>
     }]
   }
   Anim Dynamics - Bone: dyn_jacket_sleev_01_r(ComponentPose=@n3, Alpha=0.2, ExternalForce="0,0,0") {
    angularSpringConstant=200
    bAngularSpring=true
    bGravityOverrideInSimSpace=true
    bLinearSpring=true
    bUseGravityOverride=true
    lODThreshold=1
    linearSpringConstant=50
    simulationSpace=RootRelative
    alphaBoolBlend {blendOption=Linear}
    boundBone {boneName=dyn_jacket_sleev_01_r}
    gravityOverride {z=-980}
    physicsBodyDefinitions [{
      sphereCollisionRadius=0
      boundBone {boneName=dyn_jacket_sleev_01_r}
      boxExtents=<20 20 20>
      constraintSetup {
       angularTargetAxis=AxisY
       angularLimitsMax=<15 15 15>
       angularLimitsMin=<-15 -15 -15>
       angularTarget=<0 0 0>
       linearAxesMax=<0 3 1>
       linearAxesMin=<-10 0 -1>
      }
      localJointOffset=<-1 -1 -1>
     }]
   }
   Anim Dynamics - Bone: ring_01!(ComponentPose, Alpha=1, GravityOverride:Vector="0,0,-980", ExternalForce="0,0,0") -> (Pose) {
    angularBiasOverride=1
    angularDampingOverride=0.999
    angularSpringConstant=1000
    bAngularSpring=true
    bOverrideAngularDamping=true
    bUseGravityOverride=true
    bUsePlanarLimit=false
    simulationSpace=RootRelative
    alphaBoolBlend {blendInTime=0.5, blendOption=Linear, blendOutTime=0.7}
    boundBone {boneName=ring_01}
    componentAppliedLinearAccClamp=<3000 3000 3000>
    componentLinearAccScale=<0.3 0.3 0.3>
    componentLinearVelScale=<0.4 0.4 0.4>
    physicsBodyDefinitions [{
      collisionType=CustomSphere
      sphereCollisionRadius=2.5
      boundBone {boneName=ring_01}
      boxExtents=<6 6 6>
      constraintSetup {angularLimitsMax=<0 0 80>, angularLimitsMin=<0 1 0>, angularTarget=<0 0 0>}
      localJointOffset=<1 0 0>
     }]
   }
   Anim Dynamics - Bone: ring_01!(ComponentPose, bAlphaBoolEnabled:bool=True, ExternalForce="0,0,0") -> (Pose) {
    alphaInputType=Bool
    angularDampingOverride=0.99
    bOverrideAngularDamping=true
    bUseGravityOverride=true
    bUsePlanarLimit=false
    bUseSphericalLimits=true
    linearDampingOverride=0.9999
    simulationSpace=RootRelative
    alphaBoolBlend {blendInTime=0.1, blendOption=Linear, blendOutTime=0.2}
    boundBone {boneName=ring_01}
    componentAppliedLinearAccClamp=<200 200 200>
    componentLinearAccScale=<0.5 0.5 0.5>
    componentLinearVelScale=<0.6 0.6 6>
    gravityOverride {z=-400}
    physicsBodyDefinitions [{
      boundBone {boneName=ring_01}
      boxExtents=<8 8 8>
      constraintSetup {angularLimitsMax=<0 5 35>, angularLimitsMin=<0 -5 0>, angularTarget=<0 0 0>}
      localJointOffset=<1 0 0>
     }]
    sphericalLimits [{limitRadius=10, drivingBone {boneName=thigh_r}, sphereLocalOffset=<8 0 0>}]
   }
   n4 = Anim Dynamics - Bone: dyn_jacket_sleev_01_l(Alpha=0.5, ExternalForce="0,0,0") {
    angularDampingOverride=0.9
    angularSpringConstant=175
    bAngularSpring=true
    bGravityOverrideInSimSpace=true
    bLinearSpring=true
    bOverrideAngularDamping=true
    bOverrideLinearDamping=true
    gravityScale=0
    linearDampingOverride=0.995
    linearSpringConstant=50
    simulationSpace=Actor
    alphaBoolBlend {blendOption=Linear}
    boundBone {boneName=dyn_jacket_sleev_01_l}
    componentAppliedLinearAccClamp=<200 200 200>
    componentLinearAccScale=<0.1 0.1 0.1>
    componentLinearVelScale=<0.2 0.2 0.2>
    physicsBodyDefinitions [{
      collisionType=InnerSphere
      sphereCollisionRadius=0
      boundBone {boneName=dyn_jacket_sleev_01_l}
      boxExtents=<30 30 30>
      constraintSetup {angularLimitsMax=<0 0 0>, angularLimitsMin=<0 0 0>, angularTarget=<0 0 0>, linearAxesMax=<1 2 1>, linearAxesMin=<-5 0 -1>}
      localJointOffset=<0.5 0.5 0.5>
     }]
   }
   Anim Dynamics - Bone: dyn_jacket_sleev_01_r(ComponentPose=@n4, Alpha=0.5, ExternalForce="0,0,0") {
    angularDampingOverride=0.9
    angularSpringConstant=175
    bAngularSpring=true
    bGravityOverrideInSimSpace=true
    bLinearSpring=true
    bOverrideAngularDamping=true
    bOverrideLinearDamping=true
    gravityScale=0
    linearDampingOverride=0.995
    linearSpringConstant=50
    simulationSpace=Actor
    alphaBoolBlend {blendOption=Linear}
    boundBone {boneName=dyn_jacket_sleev_01_r}
    componentAppliedLinearAccClamp=<200 200 200>
    componentLinearAccScale=<0.1 0.1 0.1>
    componentLinearVelScale=<0.2 0.2 0.2>
    physicsBodyDefinitions [{
      collisionType=InnerSphere
      sphereCollisionRadius=0
      boundBone {boneName=dyn_jacket_sleev_01_r}
      boxExtents=<30 30 30>
      constraintSetup {angularLimitsMax=<0 0 0>, angularLimitsMin=<0 0 0>, angularTarget=<0 0 0>, linearAxesMax=<1 2 1>, linearAxesMin=<-5 0 -1>}
      localJointOffset=<0.5 0.5 0.5>
     }]
   }
   n5 = Anim Dynamics - Bone: rocket_02(Alpha=0.9, ExternalForce="0,0,0") {
    angularDampingOverride=0.9
    angularSpringConstant=175
    bAngularSpring=true
    bGravityOverrideInSimSpace=true
    bLinearSpring=true
    bOverrideAngularDamping=true
    bOverrideLinearDamping=true
    gravityScale=0
    linearDampingOverride=0.995
    linearSpringConstant=50
    simulationSpace=BoneRelative
    alphaBoolBlend {blendOption=Linear}
    boundBone {boneName=rocket_02}
    componentAppliedLinearAccClamp=<200 200 200>
    componentLinearAccScale=<0.1 0.1 0.1>
    componentLinearVelScale=<0.2 0.2 0.2>
    physicsBodyDefinitions [{
      collisionType=InnerSphere
      sphereCollisionRadius=0
      boundBone {boneName=rocket_02}
      boxExtents=<50 50 50>
      constraintSetup {angularLimitsMax=<0 0 0>, angularLimitsMin=<0 0 0>, angularTarget=<0 0 0>, linearAxesMax=<1 0 1>, linearAxesMin=<-1 -1 -1>}
      localJointOffset=<-2 -2 -2>
     }]
    relativeSpaceBone {boneName=Melee_TwoHanded_Handle}
   }
   n6 = Anim Dynamics - Bone: fireworks_01(ComponentPose=@n5, Alpha=0.9, ExternalForce="0,0,0") {
    angularDampingOverride=0.9
    angularSpringConstant=195
    bAngularSpring=true
    bGravityOverrideInSimSpace=true
    bLinearSpring=true
    bOverrideAngularDamping=true
    bOverrideLinearDamping=true
    gravityScale=0
    linearDampingOverride=0.995
    linearSpringConstant=50
    simulationSpace=BoneRelative
    alphaBoolBlend {blendOption=Linear}
    boundBone {boneName=fireworks_01}
    componentAppliedLinearAccClamp=<200 200 200>
    componentLinearAccScale=<0.1 0.1 0.1>
    componentLinearVelScale=<0.2 0.2 0.2>
    physicsBodyDefinitions [{
      sphereCollisionRadius=0
      boundBone {boneName=fireworks_01}
      boxExtents=<50 50 50>
      constraintSetup {angularLimitsMax=<0 0 0>, angularLimitsMin=<0 0 0>, angularTarget=<0 0 0>, linearAxesMax=<1 0 1>, linearAxesMin=<-1 -1 -1>}
      localJointOffset=<-2 -2 -2>
     }]
    relativeSpaceBone {boneName=Melee_TwoHanded_Handle}
   }
   n7 = Anim Dynamics - Bone: rocket_01(ComponentPose=@n6, Alpha=0.9, ExternalForce="0,0,0") {
    angularDampingOverride=0.9
    angularSpringConstant=195
    bAngularSpring=true
    bGravityOverrideInSimSpace=true
    bLinearSpring=true
    bOverrideAngularDamping=true
    bOverrideLinearDamping=true
    gravityScale=0
    linearDampingOverride=0.995
    linearSpringConstant=50
    simulationSpace=BoneRelative
    alphaBoolBlend {blendOption=Linear}
    boundBone {boneName=rocket_01}
    componentAppliedLinearAccClamp=<200 200 200>
    componentLinearAccScale=<0.1 0.1 0.1>
    componentLinearVelScale=<0.2 0.2 0.2>
    physicsBodyDefinitions [{
      sphereCollisionRadius=0
      boundBone {boneName=rocket_01}
      boxExtents=<50 50 50>
      constraintSetup {angularLimitsMax=<0 0 0>, angularLimitsMin=<0 0 0>, angularTarget=<0 0 0>, linearAxesMax=<1 0 1>, linearAxesMin=<-1 -1 -1>}
      localJointOffset=<-2 -2 -2>
     }]
    relativeSpaceBone {boneName=Melee_TwoHanded_Handle}
   }
   n8 = Anim Dynamics - Bone: fireworks_ball_01(ComponentPose=@n7, Alpha=0.9, ExternalForce="0,0,0") {
    angularDampingOverride=0.999
    angularSpringConstant=175
    bAngularSpring=true
    bGravityOverrideInSimSpace=true
    bLinearSpring=true
    bOverrideAngularDamping=true
    bOverrideLinearDamping=true
    gravityScale=0
    linearDampingOverride=0.995
    linearSpringConstant=50
    simulationSpace=BoneRelative
    alphaBoolBlend {blendOption=Linear}
    boundBone {boneName=fireworks_ball_01}
    componentAppliedLinearAccClamp=<200 200 200>
    componentLinearAccScale=<0.1 0.1 0.1>
    componentLinearVelScale=<0.2 0.2 0.2>
    physicsBodyDefinitions [{
      sphereCollisionRadius=0
      boundBone {boneName=fireworks_ball_01}
      boxExtents=<50 50 50>
      constraintSetup {angularLimitsMax=<0 0 0>, angularLimitsMin=<0 0 0>, angularTarget=<0 0 0>, linearAxesMax=<1 1 0>, linearAxesMin=<-1 -1 -1>}
      localJointOffset=<-2 -2 -2>
     }]
    relativeSpaceBone {boneName=Melee_TwoHanded_Handle}
   }
   Anim Dynamics - Bone: fireworks_02(ComponentPose=@n8, Alpha=0.9, ExternalForce="0,0,0") {
    angularDampingOverride=0.9
    angularSpringConstant=175
    bAngularSpring=true
    bGravityOverrideInSimSpace=true
    bLinearSpring=true
    bOverrideAngularDamping=true
    bOverrideLinearDamping=true
    gravityScale=0
    linearDampingOverride=0.995
    linearSpringConstant=50
    simulationSpace=BoneRelative
    alphaBoolBlend {blendOption=Linear}
    boundBone {boneName=fireworks_02}
    componentAppliedLinearAccClamp=<200 200 200>
    componentLinearAccScale=<0.1 0.1 0.1>
    componentLinearVelScale=<0.2 0.2 0.2>
    physicsBodyDefinitions [{
      sphereCollisionRadius=0
      boundBone {boneName=fireworks_02}
      boxExtents=<50 50 50>
      constraintSetup {angularLimitsMax=<0 0 0>, angularLimitsMin=<0 0 0>, angularTarget=<0 0 0>, linearAxesMax=<1 0 1>, linearAxesMin=<-1 -1 -1>}
      localJointOffset=<-2 -2 -2>
     }]
    relativeSpaceBone {boneName=Melee_TwoHanded_Handle}
   }
   n9 = Anim Dynamics - Bone: dyn_jacket_sleev_01_l!(ComponentPose, bAlphaBoolEnabled:bool=True, ExternalForce="0,0,0") -> (Pose) {
    alphaInputType=Bool
    angularDampingOverride=1
    angularSpringConstant=200
    bAngularSpring=true
    bLinearSpring=true
    bOverrideAngularDamping=true
    bOverrideLinearDamping=true
    bUsePlanarLimit=false
    linearDampingOverride=1
    linearSpringConstant=90
    simulationSpace=BoneRelative
    alphaBoolBlend {blendInTime=0.1, blendOption=Linear, blendOutTime=0.2}
    boundBone {boneName=dyn_jacket_sleev_01_l}
    componentAppliedLinearAccClamp=<200 200 200>
    componentLinearAccScale=<0.1 0.1 0.1>
    componentLinearVelScale=<0.2 0.2 0.2>
    gravityOverride {z=-980}
    physicsBodyDefinitions [{
      boundBone {boneName=dyn_jacket_sleev_01_l}
      boxExtents=<15 15 15>
      constraintSetup {angularLimitsMax=<0 5 5>, angularLimitsMin=<0 -5 -5>, angularTarget=<0 0 0>, linearAxesMax=<0 3 0>, linearAxesMin=<-5 0 0>}
      localJointOffset=<5 5 5>
     }]
    relativeSpaceBone {boneName=lowerarm_l}
   }
   Anim Dynamics - Bone: dyn_jacket_sleev_01_r!(ComponentPose=@n9, bAlphaBoolEnabled:bool=True, ExternalForce="0,0,0") -> (Pose) {
    alphaInputType=Bool
    angularDampingOverride=1
    angularSpringConstant=200
    bAngularSpring=true
    bLinearSpring=true
    bOverrideAngularDamping=true
    bOverrideLinearDamping=true
    bUsePlanarLimit=false
    linearDampingOverride=1
    linearSpringConstant=90
    simulationSpace=BoneRelative
    alphaBoolBlend {blendInTime=0.1, blendOption=Linear, blendOutTime=0.2}
    boundBone {boneName=dyn_jacket_sleev_01_r}
    componentAppliedLinearAccClamp=<200 200 200>
    componentLinearAccScale=<0.1 0.1 0.1>
    componentLinearVelScale=<0.2 0.2 0.2>
    gravityOverride {z=-980}
    physicsBodyDefinitions [{
      boundBone {boneName=dyn_jacket_sleev_01_r}
      boxExtents=<15 15 15>
      constraintSetup {angularLimitsMax=<0 5 5>, angularLimitsMin=<0 -5 -5>, angularTarget=<0 0 0>, linearAxesMax=<0 3 0>, linearAxesMin=<-5 0 0>}
      localJointOffset=<5 5 5>
     }]
    relativeSpaceBone {boneName=lowerarm_r}
   }
   "Constraint - ring_01_MSTR (PPPPPPPPPP)"!(
    ComponentPose
    Alpha=1
    ConstraintWeights_0=0.2038
    ConstraintWeights_1:float=0.0001
    ConstraintWeights_2:float=0
    ConstraintWeights_3:float=0.3965
    ConstraintWeights_4:float=0.2431
    ConstraintWeights_5:float=0.0092
    ConstraintWeights_6:float=0.004
    ConstraintWeights_7:float=0.0027
    ConstraintWeights_8:float=0.1382
    ConstraintWeights_9:float=0.0023
   ) -> (Pose) {
    alphaBoolBlend {blendOption=Linear}
    boneToModify {boneName=ring_01_MSTR}
    constraintSetup [
     {targetBone {boneName=spine_05}}
     {targetBone {boneName=clavicle_l}}
     {targetBone {boneName=clavicle_r}}
     {targetBone {boneName=neck_01}}
     {targetBone {boneName=neck_02}}
     {targetBone {boneName=FACIAL_C_UnderChin}}
     {targetBone {boneName=FACIAL_L_UnderChin}}
     {targetBone {boneName=FACIAL_R_UnderChin}}
     {targetBone {boneName=FACIAL_C_Jawline}}
     {targetBone {boneName=FACIAL_L_Jawline}}
    ]
    constraintWeights [1,1,1,1,1,1,1,1,1,1]
   }
   Constraint!(
    ComponentPose
    Alpha=1
    ConstraintWeights_0=0.2039
    ConstraintWeights_1:float=0.3966
    ConstraintWeights_2:float=0.2431
    ConstraintWeights_3:float=0.0092
    ConstraintWeights_4:float=0.004
    ConstraintWeights_5:float=0.0027
    ConstraintWeights_6:float=0.1382
    ConstraintWeights_7:float=0.0023
   ) -> (Pose) {
    alphaBoolBlend {blendOption=Linear}
    constraintSetup [
     {targetBone {boneName=spine_05}}
     {targetBone {boneName=neck_01}}
     {targetBone {boneName=neck_02}}
     {targetBone {boneName=FACIAL_C_UnderChin}}
     {targetBone {boneName=FACIAL_L_UnderChin}}
     {targetBone {boneName=FACIAL_R_UnderChin}}
     {targetBone {boneName=FACIAL_C_Jawline}}
     {targetBone {boneName=FACIAL_L_Jawline}}
    ]
    constraintWeights [1,1,1,1,1,1,1,1]
    boneToModify=~
   }
   n10 = Anim Dynamics - Bone: dyn_jacket_01(ComponentPose=Local To Component(LocalPose=Copy Pose From Mesh() {bUseAttachedParent=true}), Alpha=0.5, ExternalForce="0,0,0") {
    angularDampingOverride=0.8
    angularSpringConstant=175
    bAngularSpring=true
    bGravityOverrideInSimSpace=true
    bLinearSpring=true
    bOverrideAngularDamping=true
    bOverrideLinearDamping=true
    gravityScale=0
    linearDampingOverride=0.995
    linearSpringConstant=80
    simulationSpace=RootRelative
    alphaBoolBlend {blendOption=Linear}
    boundBone {boneName=dyn_jacket_01}
    componentAppliedLinearAccClamp=<400 400 400>
    componentLinearAccScale=<0.1 0.1 0.1>
    componentLinearVelScale=<0.2 0.2 0.2>
    physicsBodyDefinitions [{
      collisionType=InnerSphere
      sphereCollisionRadius=0
      boundBone {boneName=dyn_jacket_01}
      boxExtents=<30 30 30>
      constraintSetup {angularLimitsMax=<0 0 0>, angularLimitsMin=<0 0 0>, angularTarget=<0 0 0>, linearAxesMax=<1 4 1>, linearAxesMin=<-1 0 -1>}
      localJointOffset=<-2 0 0>
     }]
   }
   n11 = Anim Dynamics - Bone: dyn_jacket_01_l(ComponentPose=@n10, Alpha=0.5, ExternalForce="0,0,0") {
    angularDampingOverride=0.8
    angularSpringConstant=175
    bAngularSpring=true
    bGravityOverrideInSimSpace=true
    bLinearSpring=true
    bOverrideAngularDamping=true
    bOverrideLinearDamping=true
    gravityScale=0
    linearDampingOverride=0.995
    linearSpringConstant=80
    simulationSpace=RootRelative
    alphaBoolBlend {blendOption=Linear}
    boundBone {boneName=dyn_jacket_01_l}
    componentAppliedLinearAccClamp=<400 400 400>
    componentLinearAccScale=<0.1 0.1 0.1>
    componentLinearVelScale=<0.2 0.2 0.2>
    physicsBodyDefinitions [{
      collisionType=InnerSphere
      sphereCollisionRadius=0
      boundBone {boneName=dyn_jacket_01_l}
      boxExtents=<30 30 30>
      constraintSetup {angularLimitsMax=<0 0 0>, angularLimitsMin=<0 0 0>, angularTarget=<0 0 0>, linearAxesMax=<1 0 0>, linearAxesMin=<-1 -2 -4>}
      localJointOffset=<-2 0 0>
     }]
   }
   n12 = Anim Dynamics - Bone: dyn_jacket_01_r(ComponentPose=@n11, Alpha=0.5, ExternalForce="0,0,0") {
    angularDampingOverride=0.8
    angularSpringConstant=175
    bAngularSpring=true
    bGravityOverrideInSimSpace=true
    bLinearSpring=true
    bOverrideAngularDamping=true
    bOverrideLinearDamping=true
    gravityScale=0
    linearDampingOverride=0.995
    linearSpringConstant=80
    simulationSpace=RootRelative
    alphaBoolBlend {blendOption=Linear}
    boundBone {boneName=dyn_jacket_01_r}
    componentAppliedLinearAccClamp=<400 400 400>
    componentLinearAccScale=<0.1 0.1 0.1>
    componentLinearVelScale=<0.2 0.2 0.2>
    physicsBodyDefinitions [{
      collisionType=InnerSphere
      sphereCollisionRadius=0
      boundBone {boneName=dyn_jacket_01_r}
      boxExtents=<30 30 30>
      constraintSetup {angularLimitsMax=<0 0 0>, angularLimitsMin=<0 0 0>, angularTarget=<0 0 0>, linearAxesMax=<1 0 4>, linearAxesMin=<-1 -2 0>}
      localJointOffset=<-2 0 0>
     }]
   }
   n13 = Anim Dynamics - Bone: dyn_sleeve_dn_r!(ComponentPose=@n12, Alpha=1, GravityScale:float=1, ExternalForce="0,0,0") -> (Pose) {
    angularDampingOverride=0.99
    angularSpringConstant=1000
    bAngularSpring=true
    bGravityOverrideInSimSpace=true
    bOverrideAngularDamping=true
    bUseGravityOverride=true
    simulationSpace=RootRelative
    alphaBoolBlend {blendOption=Linear}
    boundBone {boneName=dyn_sleeve_dn_r}
    componentAppliedLinearAccClamp=<3000 3000 3000>
    componentLinearAccScale=<0.2 0.2 0.2>
    componentLinearVelScale=<0.3 0.3 0.3>
    gravityOverride {z=-500}
    physicsBodyDefinitions [{
      sphereCollisionRadius=0
      boundBone {boneName=dyn_sleeve_dn_r}
      boxExtents=<20 20 10>
      constraintSetup {angularLimitsMax=<3 5 5>, angularLimitsMin=<-3 0 -5>, angularTarget=<0 0 0>}
      localJointOffset=<5 0 0>
     }]
    relativeSpaceBone {boneName=lowerarm_r}
   }
   n14 = Anim Dynamics - Bone: dyn_sleeve_ft_r!(ComponentPose=@n13, Alpha=1, GravityScale:float=1, ExternalForce="0,0,0") -> (Pose) {
    angularDampingOverride=0.9999
    angularSpringConstant=500
    bGravityOverrideInSimSpace=true
    bOverrideAngularDamping=true
    bUseGravityOverride=true
    simulationSpace=RootRelative
    alphaBoolBlend {blendOption=Linear}
    boundBone {boneName=dyn_sleeve_ft_r}
    componentAppliedLinearAccClamp=<3000 3000 3000>
    componentLinearAccScale=<0.1 0.1 0.1>
    componentLinearVelScale=<0.3 0.3 0.3>
    gravityOverride {z=-500}
    physicsBodyDefinitions [{
      sphereCollisionRadius=0
      boundBone {boneName=dyn_sleeve_ft_r}
      boxExtents=<10 20 20>
      constraintSetup {angularLimitsMax=<3 0 2>, angularLimitsMin=<-3 -5 -2>, angularTarget=<0 0 0>}
      localJointOffset=<5 1 2>
     }]
    relativeSpaceBone {boneName=upperarm_r}
   }
   n15 = Anim Dynamics - Bone: dyn_sleeve_bk_r!(ComponentPose=@n14, Alpha=1, GravityScale:float=1, ExternalForce="0,0,0") -> (Pose) {
    angularDampingOverride=0.9999
    angularSpringConstant=500
    bGravityOverrideInSimSpace=true
    bOverrideAngularDamping=true
    bUseGravityOverride=true
    simulationSpace=RootRelative
    alphaBoolBlend {blendOption=Linear}
    boundBone {boneName=dyn_sleeve_bk_r}
    componentAppliedLinearAccClamp=<3000 3000 3000>
    componentLinearAccScale=<0.1 0.1 0.1>
    componentLinearVelScale=<0.3 0.3 0.3>
    gravityOverride {z=-500}
    physicsBodyDefinitions [{
      sphereCollisionRadius=0
      boundBone {boneName=dyn_sleeve_bk_r}
      boxExtents=<10 20 20>
      constraintSetup {angularLimitsMax=<3 0 2>, angularLimitsMin=<-3 -5 -2>, angularTarget=<0 0 0>}
      localJointOffset=<5 -1 2>
     }]
    relativeSpaceBone {boneName=upperarm_r}
   }
   n16 = Anim Dynamics - Bone: dyn_sleeve_up_r!(ComponentPose=@n15, Alpha=1, GravityScale:float=1, ExternalForce="0,0,0") -> (Pose) {
    angularDampingOverride=1
    angularSpringConstant=500
    bGravityOverrideInSimSpace=true
    bOverrideAngularDamping=true
    bUseGravityOverride=true
    simulationSpace=BoneRelative
    alphaBoolBlend {blendOption=Linear}
    boundBone {boneName=dyn_sleeve_up_r}
    componentAppliedLinearAccClamp=<3000 3000 3000>
    componentLinearAccScale=<0.1 0.1 0.1>
    componentLinearVelScale=<0.2 0.2 0.2>
    gravityOverride {y=500, z=50}
    physicsBodyDefinitions [{
      sphereCollisionRadius=0
      boundBone {boneName=dyn_sleeve_up_r}
      boxExtents=<15 20 20>
      constraintSetup {angularLimitsMax=<0 0 5>, angularLimitsMin=<0 -8 -5>, angularTarget=<0 0 0>}
      localJointOffset=<5 0 0>
     }]
    relativeSpaceBone {boneName=lowerarm_twist_02_l}
   }
   n17 = Anim Dynamics - Bone: dyn_sleeve_dn_l!(ComponentPose=@n16, Alpha=1, GravityScale:float=1, ExternalForce="0,0,0") -> (Pose) {
    angularDampingOverride=0.99
    angularSpringConstant=1000
    bAngularSpring=true
    bGravityOverrideInSimSpace=true
    bOverrideAngularDamping=true
    bUseGravityOverride=true
    simulationSpace=RootRelative
    alphaBoolBlend {blendOption=Linear}
    boundBone {boneName=dyn_sleeve_dn_l}
    componentAppliedLinearAccClamp=<3000 3000 3000>
    componentLinearAccScale=<0.2 0.2 0.2>
    componentLinearVelScale=<0.3 0.3 0.3>
    gravityOverride {z=-500}
    physicsBodyDefinitions [{
      sphereCollisionRadius=0
      boundBone {boneName=dyn_sleeve_dn_l}
      boxExtents=<20 20 10>
      constraintSetup {angularLimitsMax=<3 5 5>, angularLimitsMin=<-3 0 -5>, angularTarget=<0 0 0>}
      localJointOffset=<5 0 0>
     }]
    relativeSpaceBone {boneName=lowerarm_r}
   }
   n18 = Anim Dynamics - Bone: dyn_sleeve_ft_l!(ComponentPose=@n17, Alpha=1, GravityScale:float=1, ExternalForce="0,0,0") -> (Pose) {
    angularDampingOverride=0.9999
    angularSpringConstant=500
    bGravityOverrideInSimSpace=true
    bOverrideAngularDamping=true
    bUseGravityOverride=true
    simulationSpace=RootRelative
    alphaBoolBlend {blendOption=Linear}
    boundBone {boneName=dyn_sleeve_ft_l}
    componentAppliedLinearAccClamp=<3000 3000 3000>
    componentLinearAccScale=<0.1 0.1 0.1>
    componentLinearVelScale=<0.3 0.3 0.3>
    gravityOverride {z=-500}
    physicsBodyDefinitions [{
      sphereCollisionRadius=0
      boundBone {boneName=dyn_sleeve_ft_l}
      boxExtents=<10 20 20>
      constraintSetup {angularLimitsMax=<3 0 2>, angularLimitsMin=<-3 -5 -2>, angularTarget=<0 0 0>}
      localJointOffset=<5 1 2>
     }]
    relativeSpaceBone {boneName=upperarm_r}
   }
   n19 = Anim Dynamics - Bone: dyn_sleeve_bk_l!(ComponentPose=@n18, Alpha=1, GravityScale:float=1, ExternalForce="0,0,0") -> (Pose) {
    angularDampingOverride=0.9999
    angularSpringConstant=500
    bGravityOverrideInSimSpace=true
    bOverrideAngularDamping=true
    bUseGravityOverride=true
    simulationSpace=RootRelative
    alphaBoolBlend {blendOption=Linear}
    boundBone {boneName=dyn_sleeve_bk_l}
    componentAppliedLinearAccClamp=<3000 3000 3000>
    componentLinearAccScale=<0.1 0.1 0.1>
    componentLinearVelScale=<0.3 0.3 0.3>
    gravityOverride {z=-500}
    physicsBodyDefinitions [{
      sphereCollisionRadius=0
      boundBone {boneName=dyn_sleeve_bk_l}
      boxExtents=<10 20 20>
      constraintSetup {angularLimitsMax=<3 0 2>, angularLimitsMin=<-3 -5 -2>, angularTarget=<0 0 0>}
      localJointOffset=<5 -1 2>
     }]
    relativeSpaceBone {boneName=upperarm_r}
   }
   n20 = Anim Dynamics - Bone: dyn_sleeve_up_l!(ComponentPose=@n19, Alpha=1, GravityScale:float=1, ExternalForce="0,0,0") -> (Pose) {
    angularDampingOverride=1
    angularSpringConstant=500
    bGravityOverrideInSimSpace=true
    bOverrideAngularDamping=true
    bUseGravityOverride=true
    simulationSpace=BoneRelative
    alphaBoolBlend {blendOption=Linear}
    boundBone {boneName=dyn_sleeve_up_l}
    componentAppliedLinearAccClamp=<3000 3000 3000>
    componentLinearAccScale=<0.1 0.1 0.1>
    componentLinearVelScale=<0.2 0.2 0.2>
    gravityOverride {y=500, z=-50}
    physicsBodyDefinitions [{
      sphereCollisionRadius=0
      boundBone {boneName=dyn_sleeve_up_l}
      boxExtents=<15 20 20>
      constraintSetup {angularLimitsMax=<0 0 5>, angularLimitsMin=<0 -8 -5>, angularTarget=<0 0 0>}
      localJointOffset=<5 0 0>
     }]
    relativeSpaceBone {boneName=lowerarm_twist_02_r}
   }
   n21 = Anim Dynamics - Bone: dyn_pants_sleeve_dn_r!(ComponentPose=@n20, Alpha=1, GravityScale:float=1, ExternalForce="0,0,0") -> (Pose) {
    angularDampingOverride=0.99
    angularSpringConstant=1000
    bAngularSpring=true
    bGravityOverrideInSimSpace=true
    bOverrideAngularDamping=true
    bUseGravityOverride=true
    simulationSpace=RootRelative
    alphaBoolBlend {blendOption=Linear}
    boundBone {boneName=dyn_pants_sleeve_dn_r}
    componentAppliedLinearAccClamp=<3000 3000 3000>
    componentLinearAccScale=<0.2 0.2 0.2>
    componentLinearVelScale=<0.3 0.3 0.3>
    gravityOverride {z=-500}
    physicsBodyDefinitions [{
      sphereCollisionRadius=0
      boundBone {boneName=dyn_pants_sleeve_dn_r}
      boxExtents=<20 20 10>
      constraintSetup {angularLimitsMax=<3 5 5>, angularLimitsMin=<-3 0 -5>, angularTarget=<0 0 0>}
      localJointOffset=<5 0 0>
     }]
    relativeSpaceBone {boneName=lowerarm_r}
   }
   n22 = Anim Dynamics - Bone: dyn_pants_sleeve_ft_r!(ComponentPose=@n21, Alpha=1, GravityScale:float=1, ExternalForce="0,0,0") -> (Pose) {
    angularDampingOverride=0.9999
    angularSpringConstant=500
    bGravityOverrideInSimSpace=true
    bOverrideAngularDamping=true
    bUseGravityOverride=true
    simulationSpace=RootRelative
    alphaBoolBlend {blendOption=Linear}
    boundBone {boneName=dyn_pants_sleeve_ft_r}
    componentAppliedLinearAccClamp=<3000 3000 3000>
    componentLinearAccScale=<0.1 0.1 0.1>
    componentLinearVelScale=<0.3 0.3 0.3>
    gravityOverride {z=-500}
    physicsBodyDefinitions [{
      sphereCollisionRadius=0
      boundBone {boneName=dyn_pants_sleeve_ft_r}
      boxExtents=<10 20 20>
      constraintSetup {angularLimitsMax=<3 0 2>, angularLimitsMin=<-3 -5 -2>, angularTarget=<0 0 0>}
      localJointOffset=<5 1 2>
     }]
    relativeSpaceBone {boneName=upperarm_r}
   }
   n23 = Anim Dynamics - Bone: dyn_pants_sleeve_bk_r!(ComponentPose=@n22, Alpha=1, GravityScale:float=1, ExternalForce="0,0,0") -> (Pose) {
    angularDampingOverride=0.9999
    angularSpringConstant=500
    bGravityOverrideInSimSpace=true
    bOverrideAngularDamping=true
    bUseGravityOverride=true
    simulationSpace=RootRelative
    alphaBoolBlend {blendOption=Linear}
    boundBone {boneName=dyn_pants_sleeve_bk_r}
    componentAppliedLinearAccClamp=<3000 3000 3000>
    componentLinearAccScale=<0.1 0.1 0.1>
    componentLinearVelScale=<0.3 0.3 0.3>
    gravityOverride {z=-500}
    physicsBodyDefinitions [{
      sphereCollisionRadius=0
      boundBone {boneName=dyn_pants_sleeve_bk_r}
      boxExtents=<10 20 20>
      constraintSetup {angularLimitsMax=<3 0 2>, angularLimitsMin=<-3 -2 -2>, angularTarget=<0 0 0>}
      localJointOffset=<5 -1 2>
     }]
    relativeSpaceBone {boneName=upperarm_r}
   }
   n24 = Anim Dynamics - Bone: dyn_pants_sleeve_up_r!(ComponentPose=@n23, Alpha=1, GravityScale:float=1, ExternalForce="0,0,0") -> (Pose) {
    angularDampingOverride=1
    angularSpringConstant=500
    bGravityOverrideInSimSpace=true
    bOverrideAngularDamping=true
    bUseGravityOverride=true
    simulationSpace=BoneRelative
    alphaBoolBlend {blendOption=Linear}
    boundBone {boneName=dyn_pants_sleeve_up_r}
    componentAppliedLinearAccClamp=<3000 3000 3000>
    componentLinearAccScale=<0.1 0.1 0.1>
    componentLinearVelScale=<0.2 0.2 0.2>
    gravityOverride {y=500, z=50}
    physicsBodyDefinitions [{
      sphereCollisionRadius=0
      boundBone {boneName=dyn_pants_sleeve_up_r}
      boxExtents=<15 20 20>
      constraintSetup {angularLimitsMax=<0 0 5>, angularLimitsMin=<0 -8 -5>, angularTarget=<0 0 0>}
      localJointOffset=<5 0 0>
     }]
    relativeSpaceBone {boneName=calf_twist_01_r}
   }
   n25 = Anim Dynamics - Bone: dyn_pants_sleeve_dn_l!(ComponentPose=@n24, Alpha=1, GravityScale:float=1, ExternalForce="0,0,0") -> (Pose) {
    angularDampingOverride=0.99
    angularSpringConstant=1000
    bAngularSpring=true
    bGravityOverrideInSimSpace=true
    bOverrideAngularDamping=true
    bUseGravityOverride=true
    simulationSpace=RootRelative
    alphaBoolBlend {blendOption=Linear}
    boundBone {boneName=dyn_pants_sleeve_dn_l}
    componentAppliedLinearAccClamp=<3000 3000 3000>
    componentLinearAccScale=<0.2 0.2 0.2>
    componentLinearVelScale=<0.3 0.3 0.3>
    gravityOverride {z=-980}
    physicsBodyDefinitions [{
      sphereCollisionRadius=0
      boundBone {boneName=dyn_pants_sleeve_dn_l}
      boxExtents=<20 20 10>
      constraintSetup {angularLimitsMax=<3 3 3>, angularLimitsMin=<-3 -3 -3>, angularTarget=<0 0 0>}
      localJointOffset=<5 0 0>
     }]
    relativeSpaceBone {boneName=calf_twist_01_l}
    sphericalLimits [{limitRadius=5, drivingBone {boneName=foot_l}, sphereLocalOffset=<0 0 0>}]
   }
   n26 = Anim Dynamics - Bone: dyn_pants_sleeve_ft_l!(ComponentPose=@n25, Alpha=1, GravityScale:float=1, ExternalForce="0,0,0") -> (Pose) {
    angularDampingOverride=0.9999
    angularSpringConstant=500
    bGravityOverrideInSimSpace=true
    bOverrideAngularDamping=true
    bUseGravityOverride=true
    simulationSpace=RootRelative
    alphaBoolBlend {blendOption=Linear}
    boundBone {boneName=dyn_pants_sleeve_ft_l}
    componentAppliedLinearAccClamp=<3000 3000 3000>
    componentLinearAccScale=<0.1 0.1 0.1>
    componentLinearVelScale=<0.3 0.3 0.3>
    gravityOverride {z=-980}
    physicsBodyDefinitions [{
      sphereCollisionRadius=0
      boundBone {boneName=dyn_pants_sleeve_ft_l}
      boxExtents=<10 20 20>
      constraintSetup {angularLimitsMax=<3 3 3>, angularLimitsMin=<-3 -3 -3>, angularTarget=<0 0 0>}
      localJointOffset=<5 0 0>
     }]
    relativeSpaceBone {boneName=calf_twist_01_l}
    sphericalLimits [{limitRadius=5, drivingBone {boneName=foot_l}, sphereLocalOffset=<0 0 0>}]
   }
   n27 = Anim Dynamics - Bone: dyn_pants_sleeve_bk_l!(ComponentPose=@n26, Alpha=1, GravityScale:float=1, ExternalForce="0,0,0") -> (Pose) {
    angularDampingOverride=0.9999
    angularSpringConstant=500
    bGravityOverrideInSimSpace=true
    bOverrideAngularDamping=true
    bUseGravityOverride=true
    simulationSpace=RootRelative
    alphaBoolBlend {blendOption=Linear}
    boundBone {boneName=dyn_pants_sleeve_bk_l}
    componentAppliedLinearAccClamp=<3000 3000 3000>
    componentLinearAccScale=<0.1 0.1 0.1>
    componentLinearVelScale=<0.3 0.3 0.3>
    gravityOverride {z=-980}
    physicsBodyDefinitions [{
      sphereCollisionRadius=0
      boundBone {boneName=dyn_pants_sleeve_bk_l}
      boxExtents=<10 20 20>
      constraintSetup {angularLimitsMax=<3 3 3>, angularLimitsMin=<-3 -3 -3>, angularTarget=<0 0 0>}
      localJointOffset=<5 0 0>
     }]
    relativeSpaceBone {boneName=calf_twist_01_l}
    sphericalLimits [{limitRadius=5, drivingBone {boneName=foot_l}, sphereLocalOffset=<0 0 0>}]
   }
   n28 = Anim Dynamics - Bone: dyn_pants_sleeve_up_l!(ComponentPose=@n27, Alpha=1, GravityScale:float=1, ExternalForce="0,0,0") -> (Pose) {
    angularDampingOverride=1
    angularSpringConstant=500
    bGravityOverrideInSimSpace=true
    bOverrideAngularDamping=true
    bUseGravityOverride=true
    simulationSpace=RootRelative
    alphaBoolBlend {blendOption=Linear}
    boundBone {boneName=dyn_pants_sleeve_up_l}
    componentAppliedLinearAccClamp=<3000 3000 3000>
    componentLinearAccScale=<0.1 0.1 0.1>
    componentLinearVelScale=<0.2 0.2 0.2>
    gravityOverride {z=-980}
    physicsBodyDefinitions [{
      sphereCollisionRadius=0
      boundBone {boneName=dyn_pants_sleeve_up_l}
      boxExtents=<15 20 10>
      constraintSetup {angularLimitsMax=<3 3 3>, angularLimitsMin=<-3 -3 -3>, angularTarget=<0 0 0>}
      localJointOffset=<5 0 0>
     }]
    relativeSpaceBone {boneName=calf_twist_01_l}
    sphericalLimits [{limitRadius=5, drivingBone {boneName=foot_l}, sphereLocalOffset=<0 0 0>}]
   }
   n29 = "Constraint - ring_01_MSTR (PPPPPPPP)"!(
    ComponentPose=@n28
    Alpha=1
    ConstraintWeights_0=0.2052
    ConstraintWeights_1:float=0.3957
    ConstraintWeights_2:float=0.2432
    ConstraintWeights_3:float=0.0092
    ConstraintWeights_4:float=0.0041
    ConstraintWeights_5:float=0.0027
    ConstraintWeights_6:float=0.1374
    ConstraintWeights_7:float=0.0025
   ) -> (Pose) {
    alphaBoolBlend {blendOption=Linear}
    boneToModify {boneName=ring_01_MSTR}
    constraintSetup [
     {targetBone {boneName=spine_05}}
     {targetBone {boneName=neck_01}}
     {targetBone {boneName=neck_02}}
     {targetBone {boneName=FACIAL_C_UnderChin}}
     {targetBone {boneName=FACIAL_L_UnderChin}}
     {targetBone {boneName=FACIAL_R_UnderChin}}
     {targetBone {boneName=FACIAL_C_Jawline}}
     {targetBone {boneName=FACIAL_L_Jawline}}
    ]
    constraintWeights [1,1,1,1,1,1,1,1]
   }
   n30 = Anim Dynamics - Bone: dyn_pants_sleev_01_l(ComponentPose=@n29, Alpha=0.7, ExternalForce="0,0,0") {
    angularDampingOverride=0.8
    angularSpringConstant=175
    bAngularSpring=true
    bGravityOverrideInSimSpace=true
    bLinearSpring=true
    bOverrideAngularDamping=true
    bOverrideLinearDamping=true
    gravityScale=0
    linearDampingOverride=0.995
    linearSpringConstant=80
    simulationSpace=RootRelative
    alphaBoolBlend {blendOption=Linear}
    boundBone {boneName=dyn_pants_sleev_01_l}
    componentAppliedLinearAccClamp=<200 200 200>
    componentLinearAccScale=<0.1 0.1 0.1>
    componentLinearVelScale=<0.2 0.2 0.2>
    physicsBodyDefinitions [{
      collisionType=InnerSphere
      sphereCollisionRadius=0
      boundBone {boneName=dyn_pants_sleev_01_l}
      boxExtents=<30 30 30>
      constraintSetup {angularLimitsMax=<5 5 5>, angularLimitsMin=<-5 -5 -5>, angularTarget=<0 0 0>, linearAxesMax=<0 0.5 0.5>, linearAxesMin=<-5 -0.5 -0.5>}
      localJointOffset=<-2 0 0>
     }]
   }
   n31 = Anim Dynamics - Bone: dyn_pants_sleev_01_r(ComponentPose=@n30, Alpha=0.7, ExternalForce="0,0,0") {
    angularDampingOverride=0.8
    angularSpringConstant=175
    bAngularSpring=true
    bGravityOverrideInSimSpace=true
    bLinearSpring=true
    bOverrideAngularDamping=true
    bOverrideLinearDamping=true
    gravityScale=0
    linearDampingOverride=0.995
    linearSpringConstant=80
    simulationSpace=RootRelative
    alphaBoolBlend {blendOption=Linear}
    boundBone {boneName=dyn_pants_sleev_01_r}
    componentAppliedLinearAccClamp=<200 200 200>
    componentLinearAccScale=<0.1 0.1 0.1>
    componentLinearVelScale=<0.2 0.2 0.2>
    physicsBodyDefinitions [{
      collisionType=InnerSphere
      sphereCollisionRadius=0
      boundBone {boneName=dyn_pants_sleev_01_r}
      boxExtents=<30 30 30>
      constraintSetup {angularLimitsMax=<5 5 5>, angularLimitsMin=<-5 -5 -5>, angularTarget=<0 0 0>, linearAxesMax=<0 0.5 0.5>, linearAxesMin=<-5 -0.5 -0.5>}
      localJointOffset=<-2 0 0>
     }]
   }
   n32 = Anim Dynamics - Bone: ring_01!(ComponentPose=@n31, Alpha:float=1, ExternalForce="0,0,0") -> (Pose) {
    angularDampingOverride=0.9999
    angularSpringConstant=50
    bOverrideAngularDamping=true
    bUseGravityOverride=true
    bUsePlanarLimit=false
    linearDampingOverride=0.9999
    simulationSpace=RootRelative
    alphaBoolBlend {blendInTime=0.1, blendOption=Linear, blendOutTime=0.2}
    boundBone {boneName=ring_01}
    componentAppliedLinearAccClamp=<1000 1000 1000>
    componentLinearAccScale=<0.5 0.5 0.5>
    componentLinearVelScale=<0.6 0.6 0.6>
    gravityOverride {z=-980}
    physicsBodyDefinitions [{
      boundBone {boneName=ring_01}
      boxExtents=<7 7 7>
      constraintSetup {angularLimitsMax=<0 2 35>, angularLimitsMin=<0 -2 0>, angularTarget=<0 0 0>}
      localJointOffset=<1.5 0 0>
     }]
   }
   n33 = Control Rig(Source=Component To Local(ComponentPose=@n32)) {
    alphaBoolBlend {blendOption=Linear}
    controlRigAssetReference {blueprintRigClass=&/F_MED_CraneLaugh_CtrlRig.F_MED_CraneLaugh_CtrlRig_C}
   }
   Output Pose(Result=@n33) {layerGroup=None}
  ]}
 EventGraph {nodes [
   Event OnCharacterCustomizationLoaded()
   > n1 = Set Skel Mesh(Skel Mesh=GetPartSkeletalMeshComponent())
   > n2 = Get Dynamic Material Instance(SkeletalMeshComponent=@n1)
   > Set Mat Instance Dynamic(Mat Instance Dynamic=@n2)
   > n3 = GetOrCreateDynamicOverlayMaterial(MeshComponent=@n1, MaterialName=None)
   > Set Overlay Mat Instance Dynamic(Overlay Mat Instance Dynamic=@n3)
   Event BlueprintInitializeAnimation()
   > Set Skel Mesh(Skel Mesh=GetOwnerSkeletalMeshComponent())
   n4 = TryGetPawnOwner()
   > Is Valid(InputObject=@n4)
   > n5 = Cast To FortPlayerPawnAthena(Object=@n4)
   > Set PlayerPawn(PlayerPawn:Fort Player Pawn Athena Object Reference=@n5) -> (Output_Get:Fort Player Pawn Athena Object Reference)
   n6 = Event BlueprintUpdateAnimation()
   > n7 = Set Delta Time X(Delta Time X=@n6.DeltaTimeX)
   > Sequence()
   > Is Valid(InputObject=Get PlayerPawn() -> (PlayerPawn:Fort Player Pawn Athena Object Reference))
   > Set WPO Globals Function(
    Material=Get Mat Instance Dynamic()
    Overlay Material=Get Overlay Mat Instance Dynamic()
    Skeletal Mesh=Get Skel Mesh()
    "Local Noise Transform  Location"="Get Local Noise Transform  Location"() -> ("Local Noise Transform  Location":Vector)
    Local Noise Offset=Get Local Noise Offset()
   )
   > Set WPO Regional Function!(
    execute
    Material:Material Instance Dynamic Object Reference=Get Mat Instance Dynamic()
    Overlay Material:Material Instance Dynamic Object Reference=Get Overlay Mat Instance Dynamic()
    Skeletal Mesh:Skeletal Mesh Component Object Reference=Get Skel Mesh()
    Velocity Transform:Name=foot_r
    Velocity Interp Speed:double=Get Velocity Interp Speed()
    Collision Transform:Name=calf_r
    Collision Transform Axis:EAxis Enum
    Invert Collision Transform Axis:bool=false
    Delta Time X:double=@n7
    Previous Velocity Transform Location:"Vector (by ref)"=Get R1 Velocity Transform Location()
    Previous Velocity Transform Velocity:"Vector (by ref)"=Get R1 Velocity Transform Velocity()
    Region:int=1
   ) -> (then)
   > Set WPO Regional Function(
    Material=Get Mat Instance Dynamic()
    Overlay Material=Get Overlay Mat Instance Dynamic()
    Skeletal Mesh=Get Skel Mesh()
    Velocity Transform=lowearm_r
    Velocity Interp Speed=Get Velocity Interp Speed()
    Collision Transform=upperarm_r
    Invert Collision Transform Axis=false
    Delta Time X=Get Delta Time X()
    Previous Velocity Transform Location=Get R2 Velocity Transform Location()
    Previous Velocity Transform Velocity=Get R2 Velocity Transform Velocity()
    Region=2
   )
   > SetScalarParameterValue(self=Get Mat Instance Dynamic(), ParameterName=ClothSim_WindStrength, Value=SelectFloat(A=1, B=0, bPickA=Get bIsSkydiveFloating()))
  ]}
]
variables [
 Delta Time X {type=number}
 Local Noise Offset {type=Vector}
 "Local Noise Transform  Location" {type=Vector}
 Mat Instance Dynamic {type=MaterialInstanceDynamic}
 Overlay Mat Instance Dynamic {type=MaterialInstanceDynamic}
 Player Pawn {type=FortPlayerPawnAthena}
 PlayerPawn {type=FortPlayerPawnAthena}
 R1 Velocity Transform Location {type=Vector}
 R1 Velocity Transform Velocity {type=Vector}
 R2 Velocity Transform Location {type=Vector}
 R2 Velocity Transform Velocity {type=Vector}
 Skel Mesh {type=SkeletalMeshComponent}
 Velocity Interp Speed {type=number}
]
