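# Physics asset for the CraneLaugh female medium character's FaceAcc part, setting up a dynamic hanging braid with a loop.
# Capsules on the head, neck, spine, pelvis, clavicles and upper arms form a body volume around the braid's chain of spheres and capsules.
# Free angular joints connect head to loop base and loop to braid base, with locked joints after each.
# Further down, narrow swing limits widen toward the tip while twist stays tight.
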PhysicsAsset /BRCosmetics/Characters/Player/Female/Medium/Bodies/F_MED_CraneLaugh/Meshes/Parts/F_MED_CraneLaugh_FaceAcc_Physycs
bodies {
 clavicle_l [{
   shapeType=Capsule
   length=13.6
   radius=8.6
   shapeName=clavicle_l_capsule
   center=<8.7333 1.2451 0.7716>
   rotation=<rot 85 -173.2572 178.6338>
  }]
 clavicle_r [{shapeType=Capsule, length=13.6, radius=8.6, shapeName=clavicle_r_capsule, center=<-6.2175 0 0>, rotation=<rot 85 -164.2849 -169.8789>}]
 dyn_braid_00_base [{shapeType=Sphere, radius=0.9, shapeName=dyn_braid_00_base_sphere}]
 dyn_braid_01 [{shapeType=Capsule, length=0.4, radius=1.6, shapeName=dyn_braid_01_capsule, center=<1.4534 0 0>, rotation=<rot 90 0 0>}]
 dyn_braid_02 [{shapeType=Capsule, length=0, radius=1.9, shapeName=dyn_braid_02_capsule, center=<1.1466 0 0>, rotation=<rot 90 0 0>}]
 dyn_braid_03 [{shapeType=Capsule, length=3.7, radius=1.6, shapeName=dyn_braid_03_capsule, center=<3.8275 0 0>, rotation=<rot 90 0 0>}]
 dyn_braid_04 [{shapeType=Capsule, length=9.7, radius=1.6, shapeName=dyn_braid_04_capsule, center=<6.048 0 0>, rotation=<rot 90 0 0>}]
 dyn_braid_05 [{shapeType=Capsule, length=7.3, radius=1.6, shapeName=dyn_braid_05_capsule, center=<4.3919 0 0>, rotation=<rot 90 0 0>}]
 dyn_braid_06 [{shapeType=Capsule, length=6.1, radius=1.8, shapeName=dyn_braid_06_capsule, center=<4.9524 0 0>, rotation=<rot 90 0 0>}]
 dyn_braid_07 [{shapeType=Capsule, length=5.9, radius=1.7, shapeName=dyn_braid_07_capsule, center=<4.7987 0 0>, rotation=<rot 90 0 0>}]
 dyn_braid_08 [{shapeType=Capsule, length=7.6, radius=1.5, shapeName=dyn_braid_08_capsule, center=<6.1328 0 0>, rotation=<rot 90 0 0>}]
 dyn_braid_loop [{shapeType=Capsule, length=4.7, radius=3.6, shapeName=dyn_braid_loop_capsule, center=<5.1826 0 2.9684>, rotation=<rot 58.1862 -180 -180>}]
 dyn_dyn_braid_loop_base [{shapeType=Sphere, radius=0.9, shapeName=dyn_dyn_braid_loop_base_sphere}]
 head [{shapeType=Capsule, length=54.1, radius=11.7, center=<5.6297 4.1753 -0.0119>, rotation=<rot 0 53.9035 0>}]
 neck_01 [{shapeType=Capsule, length=57.1, radius=14.7, shapeName=neck_02_capsule, center=<11.6318 6.725 0>}]
 neck_02 [{shapeType=Capsule, length=2.4, radius=4.4, shapeName=neck_02_capsule, center=<0 0.1467 0>, rotation=<rot 90 0 0>}]
 pelvis [{shapeType=Capsule, length=8.2, radius=11.7, shapeName=pelvis_capsule, center=<-5.4918 -1.3048 0>}]
 spine_01 [{shapeType=Capsule, length=6.8, radius=10, shapeName=spine_01_capsule, center=<-2.3902 0.224 0>}]
 spine_02 [{shapeType=Capsule, length=5.1, radius=9.6, shapeName=spine_02_capsule, center=<0 2.0909 0>}]
 spine_03 [{shapeType=Capsule, length=4.5, radius=13.3, shapeName=spine_03_capsule, center=<4.7667 4.6843 0>}]
 spine_04 [{shapeType=Capsule, length=10.7, radius=11.1, shapeName=spine_04_capsule, center=<5.5566 1.4552 0>}]
 spine_05 [{shapeType=Capsule, length=60.3, radius=14.1, shapeName=spine_05_capsule, center=<7.7638 5.5755 0>}]
 upperarm_l [{shapeType=Capsule, length=17.9, radius=7.5, shapeName=upperarm_l_capsule, center=<11.3292 1.1817 0>, rotation=<rot 90 0 0>}]
 upperarm_r [{shapeType=Capsule, length=17.9, radius=7.5, shapeName=upperarm_r_capsule, center=<-11.3292 -1.1817 0>, rotation=<rot 90 0 0>}]
}
constraints [
 {bone1Name=dyn_braid_08, bone2Name=dyn_braid_07, swing1LimitDegrees=14, swing2LimitDegrees=10, twistLimitDegrees=5}
 {bone1Name=dyn_braid_07, bone2Name=dyn_braid_06, swing1LimitDegrees=14, swing2LimitDegrees=10, twistLimitDegrees=5}
 {bone1Name=dyn_braid_06, bone2Name=dyn_braid_05, swing1LimitDegrees=14, swing2LimitDegrees=10, twistLimitDegrees=5}
 {bone1Name=dyn_braid_05, bone2Name=dyn_braid_04, swing1LimitDegrees=12, swing2LimitDegrees=10, twistLimitDegrees=5}
 {bone1Name=dyn_braid_04, bone2Name=dyn_braid_03, swing1LimitDegrees=12, swing2LimitDegrees=10, twistLimitDegrees=5}
 {bone1Name=dyn_braid_03, bone2Name=dyn_braid_02, swing1LimitDegrees=12, swing2LimitDegrees=10, twistLimitDegrees=5}
 {bone1Name=dyn_braid_02, bone2Name=dyn_braid_01, swing1LimitDegrees=8, swing2LimitDegrees=8, twistLimitDegrees=5}
 {
  bone1Name=dyn_braid_01
  bone2Name=dyn_braid_00_base
  swing1LimitDegrees=8
  swing1Motion=Locked
  swing2LimitDegrees=8
  swing2Motion=Locked
  twistLimitDegrees=5
  twistMotion=Locked
 }
 {
  bone1Name=dyn_braid_00_base
  bone2Name=dyn_braid_loop
  swing1LimitDegrees=45
  swing1Motion=Free
  swing2LimitDegrees=45
  swing2Motion=Free
  twistLimitDegrees=45
  twistMotion=Free
 }
 {
  bone1Name=dyn_dyn_braid_loop_base
  bone2Name=head
  swing1LimitDegrees=45
  swing1Motion=Free
  swing2LimitDegrees=45
  swing2Motion=Free
  twistLimitDegrees=45
  twistMotion=Free
 }
 {
  bone1Name=dyn_braid_loop
  bone2Name=dyn_dyn_braid_loop_base
  swing1LimitDegrees=45
  swing1Motion=Locked
  swing2LimitDegrees=45
  swing2Motion=Locked
  twistLimitDegrees=45
  twistMotion=Locked
 }
]
