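# Control Rig for the F_MED CraneLaugh face accessory that procedurally poses the character's long braid from head motion.
# Two spherical pose readers track head orientation relative to spine_04 and weight translation offsets on braid and spline guide bones.
# Upper bones dyn_braid_02 and 03 aim at the first guide, then dyn_braid_04–08 are stretch-fitted to a Hermite spline through five neck/spine-parented guides.
# The Is DBNO Carrying variable is unused.
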ControlRigBlueprint /BRCosmetics/Characters/Player/Female/Medium/Bodies/F_MED_CraneLaugh/Meshes/Parts/F_MED_Crane_Laugh_FaceAcc_CtrlRig
graphs [
 RigVMFunctionLibrary {links [], nodes []}
 RigVMModel {nodes [
   RigUnit_BeginExecution = &RigUnit_BeginExecution()
   n1 = &RigUnit_SphericalPoseReader(
    ActiveRegionScaleFactors=(PositiveWidth=2,NegativeWidth=0.8,PositiveHeight=2,NegativeHeight=2)
    ActiveRegionSize=0.1
    Debug=(bDrawDebug=false,bDraw2D=false,bDrawLocalAxes=false,DebugScale=25,DebugSegments=20,DebugThickness=0.25)
    DriverAxis=(X=0,Y=1,Z=0)
    DriverItem=(Type=Bone,Name="head")
    FalloffRegionScaleFactors=(PositiveWidth=1,NegativeWidth=0.8,PositiveHeight=2,NegativeHeight=2)
    FalloffSize=0.3
    FlipHeightScaling=False
    FlipWidthScaling=False
    OptionalParentItem=(Type=Bone,Name="spine_04")
    RotationOffset=(X=90,Y=0,Z=-65)
    ExecutePin=@RigUnit_BeginExecution.ExecutePin
   )
   > &RigUnit_ModifyTransforms(
    ItemToModify="((Item=(Type=Bone,Name=\"dyn_braid_spline_00\"),Transform=(Rotation=(X=0,Y=0,Z=0,W=1),Translation=(X=0,Y=2,Z=-2),Scale3D=(X=1,Y=1,Z=1))),(Item=(Type=Bone,Name=\"dyn_braid_03\"),Transform=(Rotation=(X=0,Y=0,Z=0,W=1),Translation=(X=0,Y=0,Z=0),Scale3D=(X=1,Y=1,Z= ...(truncated)"
    WeightMinimum=0
    Weight=@n1.OutputParam
   )
   > /Script/ControlRig.RigUnit_AimItem(
    DebugSettings=(bEnabled=False,Scale=10,WorldOffset=(Rotation=(X=0,Y=0,Z=0,W=1),Translation=(X=0,Y=0,Z=0),Scale3D=(X=1,Y=1,Z=1)))
    Item=(Type=Bone,Name="dyn_braid_02")
    Primary=(Weight=1,Axis=(X=1,Y=0,Z=0),Target=(X=0,Y=0,Z=-1),Kind=Location,Space=(Type=Bone,Name="dyn_braid_spline_00"))
    Secondary=(Weight=0,Axis=(X=0,Y=0,Z=1),Target=(X=1,Y=0,Z=0),Kind=Location,Space=(Type=Bone,Name="None"))
    Weight=1
   ) {is=struct RigVMUnitNode}
   > &RigUnit_AimItem(
    DebugSettings=(bEnabled=False,Scale=10,WorldOffset=(Rotation=(X=0,Y=0,Z=0,W=1),Translation=(X=0,Y=0,Z=0),Scale3D=(X=1,Y=1,Z=1)))
    Item=(Type=Bone,Name="dyn_braid_03")
    Primary=(Weight=1,Axis=(X=1,Y=0,Z=0),Target=(X=0,Y=0,Z=-1),Kind=Location,Space=(Type=Bone,Name="dyn_braid_spline_00"))
    Secondary=(Weight=0,Axis=(X=0,Y=0,Z=1),Target=(X=1,Y=0,Z=0),Kind=Location,Space=(Type=Bone,Name="None"))
    Weight=1
   )
   > n2 = &RigUnit_SphericalPoseReader(
    ActiveRegionScaleFactors=(PositiveWidth=2.5,NegativeWidth=1.5,PositiveHeight=2.5,NegativeHeight=0.7)
    ActiveRegionSize=0.1
    Debug=(bDrawDebug=false,bDraw2D=false,bDrawLocalAxes=false,DebugScale=25,DebugSegments=20,DebugThickness=0.25)
    DriverAxis=(X=0,Y=1,Z=0)
    DriverItem=(Type=Bone,Name="head")
    FalloffRegionScaleFactors=(PositiveWidth=3,NegativeWidth=1,PositiveHeight=5,NegativeHeight=1)
    FalloffSize=0.3
    FlipHeightScaling=False
    FlipWidthScaling=False
    OptionalParentItem=(Type=Bone,Name="spine_04")
    RotationOffset=(X=135,Y=0,Z=0)
   )
   > &RigUnit_ModifyTransforms(
    ItemToModify="((Item=(Type=Bone,Name=\"dyn_braid_spline_01\"),Transform=(Rotation=(X=0,Y=0,Z=0,W=1),Translation=(X=-2,Y=0,Z=0),Scale3D=(X=1,Y=1,Z=1))),(Item=(Type=Bone,Name=\"dyn_braid_spline_02\"),Transform=(Rotation=(X=0,Y=0,Z=0,W=1),Translation=(X=-1,Y=0,Z=0),Scale3D=(X= ...(truncated)"
    WeightMinimum=0
    Weight=@n2.OutputParam
   )
   > SplineFromItems = RigVMFunctionReferenceNode(
    Items=((Type=Bone,Name="dyn_braid_spline_00"),(Type=Bone,Name="dyn_braid_spline_01"),(Type=Bone,Name="dyn_braid_spline_02"),(Type=Bone,Name="dyn_braid_spline_03"),(Type=Bone,Name="dyn_braid_spline_04"))
    Samples Per Segment=16
    Spline Mode=Hermite
   )
   n3 = RigVMDispatchNode(Values=(((),Type=Bone,Name="dyn_braid_04"),((),Type=Bone,Name="dyn_braid_05"),((),Type=Bone,Name="dyn_braid_06"),((),Type=Bone,Name="dyn_braid_07"),((),Type=Bone,Name="dyn_braid_08"))) {template="DISPATCH_RigVMDispatch_ArrayMake(in Values,out Array)"}
   > /Script/ControlRigSpline.RigUnit_FitChainToSplineCurveItemArray(
    Alignment=Stretched
    DebugSettings=(bEnabled=False,Scale=1,CurveColor=(R=1,G=1,B=0,A=1),SegmentsColor=(R=1,G=0,B=0,A=1),WorldOffset=(Rotation=(X=0,Y=0,Z=0,W=1),Translation=(X=0,Y=0,Z=0),Scale3D=(X=1,Y=1,Z=1)))
    Maximum=1
    Minimum=0
    PoleVectorPosition=(X=0,Y=0,Z=0)
    PrimaryAxis=(X=1,Y=0,Z=0)
    RotationEaseType=Linear
    Rotations=()
    SamplingPrecision=12
    SecondaryAxis=(X=0,Y=0,Z=0)
    Weight=1
    bPropagateToChildren=True
    Items=@n3.Array
    Spline=@SplineFromItems.Spline
   ) {is=struct RigVMUnitNode}
  ]}
]
hierarchy [
 root {loc=<0 0 0>, rot=<0 0 0 1>}
 pelvis {parent=root, loc=<0 -0.5301 98.0571>, rot=<0.0127 0.707 -0.0127 -0.707>}
 spine_01 {parent=pelvis, loc=<6.1208 -0.9328 0>, rot=<0 0 0.034 -0.9994>}
 spine_02 {parent=spine_01, loc=<6.5997 -0.2095 0>, rot=<0 0 0.0238 -0.9997>}
 spine_03 {parent=spine_02, loc=<6.4914 -0.1237 0>, rot=<0 0 0.0149 -0.9999>}
 spine_04 {parent=spine_03, loc=<6.9598 0.0794 0>, rot=<0 0 -0.0132 -0.9999>}
 spine_05 {parent=spine_04, loc=<6.9987 0 0>, rot=<0 0 -0.0015 -1>}
 neck_01 {parent=spine_05, loc=<11.3138 0.1529 0>, rot=<0 0 -0.2003 -0.9797>}
 neck_02 {parent=neck_01, loc=<4.5627 0 0>, rot=<0 0 0.0504 -0.9987>}
 head {parent=neck_02, loc=<5.3914 0 0>, rot=<0 0 0.115 -0.9934>}
 dyn_dyn_braid_loop_base {parent=head, loc=<12.5174 -5.4015 0>, rot=<0.5141 -0.4855 0.4855 -0.5141>}
 dyn_braid_loop {parent=dyn_dyn_braid_loop_base, loc=<0 0 0>, rot=<0 0 0 1>}
 dyn_braid_00_base {parent=dyn_braid_loop, loc=<4.5193 0 0>, rot=<0 -0.6646 0 -0.7472>}
 dyn_braid_01 {parent=dyn_braid_00_base, loc=<0 0 0>, rot=<0 0 0 1>}
 dyn_braid_02 {parent=dyn_braid_01, loc=<2.0043 0 0>, rot=<0 0.0014 0 -1>}
 dyn_braid_03 {parent=dyn_braid_02, loc=<2.3953 0 0>, rot=<0 0.0129 0 -0.9999>}
 dyn_braid_04 {parent=dyn_braid_03, loc=<7.835 0 0>, rot=<0 -0.0065 0 -1>}
 dyn_braid_05 {parent=dyn_braid_04, loc=<11.3372 0 0>, rot=<0 -0.0804 0 -0.9968>}
 dyn_braid_06 {parent=dyn_braid_05, loc=<9.1275 0 0>, rot=<0 0.0249 0 -0.9997>}
 dyn_braid_07 {parent=dyn_braid_06, loc=<9.1255 0 0>, rot=<0 0.058 0 -0.9983>}
 dyn_braid_08 {parent=dyn_braid_07, loc=<8.4908 0 0>, rot=<0 0 0 1>}
 dyn_bang_03_l {parent=head, loc=<9.3087 6.2393 -4.9367>, rot=<0.5817 0.0069 -0.0096 -0.8133>}
 dyn_bang_02_l {parent=head, loc=<9.2757 7.6386 -4.3714>, rot=<0.4678 0.0055 -0.0104 -0.8837>}
 dyn_bang_01_l {parent=head, loc=<9.2305 9.5501 -2.1095>, rot=<0.1924 0.0023 -0.0116 -0.9812>}
 dyn_bang_01_r {parent=head, loc=<9.227 9.698 1.8809>, rot=<-0.1279 -0.0015 -0.0117 -0.9917>}
 dyn_bang_02_r {parent=head, loc=<9.2786 7.5158 4.3572>, rot=<-0.4846 -0.0057 -0.0103 -0.8747>}
 dyn_braid_spline_00 {parent=neck_02, loc=<3.0809 -8.5527 0>, rot=<-0.0784 -0.7027 0.0784 -0.7027>}
 clavicle_l {parent=spine_05, loc=<5.4831 -0.1053 -3.2946>, rot=<0.0465 -0.7289 -0.009 -0.6829>}
 upperarm_l {parent=clavicle_l, loc=<9.8658 -1.2136 -0.4105>, rot=<0.0164 -0.3245 -0.0269 -0.9454>}
 clavicle_r {parent=spine_05, loc=<5.4831 -0.1053 3.2946>, rot=<0.7289 0.0465 -0.6829 0.009>}
 upperarm_r {parent=clavicle_r, loc=<-9.8658 1.2136 0.4105>, rot=<0.0164 -0.3245 -0.0269 -0.9454>}
 dyn_braid_spline_01 {parent=spine_05, loc=<10.7521 -7.9741 0>, rot=<0.0284 -0.7065 -0.0284 -0.7065>}
 dyn_braid_spline_02 {parent=spine_05, loc=<1.8551 -10.2536 0>, rot=<0.0284 -0.7065 -0.0284 -0.7065>}
 dyn_braid_spline_03 {parent=spine_04, loc=<-0.5168 -10.182 0>, rot=<0.0294 -0.7065 -0.0294 -0.7065>}
 dyn_braid_spline_04 {parent=spine_03, loc=<-1.4579 -10.4728 0>, rot=<0.0387 -0.706 -0.0387 -0.706>}
]
variables [Is DBNO Carrying {type=bool, default=False}]
