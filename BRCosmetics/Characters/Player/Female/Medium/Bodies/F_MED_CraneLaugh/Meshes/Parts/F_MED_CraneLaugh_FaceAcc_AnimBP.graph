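# Post-process-style AnimBP for the CraneLaugh face accessory, animating bangs and braid over the parent mesh's copied pose.
# AnimDynamics swings the bangs, then a neck-relative RigidBody simulation with per-state gravity runs when enabled and no backpack is worn.
# The fallback offsets braid bones, runs a Control Rig and, without a cape, adds AnimDynamics and skydive tilt.
# EventGraph updates gravity and backbling flags each frame.
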AnimBlueprint /BRCosmetics/Characters/Player/Female/Medium/Bodies/F_MED_CraneLaugh/Meshes/Parts/F_MED_CraneLaugh_FaceAcc_AnimBP
target_skeleton=&/F_MED_CraneLaugh_FaceAcc_Skeleton
graphs [
 AnimGraph {nodes [
   n1 = RigidBody!(
    ComponentPose=Local To Component(LocalPose=Use cached pose 'post bang dyn pose'())
    Alpha
    OverrideWorldGravity:Vector=Get Gravity Override()
    ExternalForce
   ) -> (Pose) {
    bOverrideWorldGravity=true
    overridePhysicsAsset=&/F_MED_CraneLaugh_FaceAcc_Physycs
    simulationSpace=BaseBoneSpace
    alphaBoolBlend {blendOption=Linear}
    baseBoneRef {boneName=neck_01}
    componentAppliedLinearAccClamp=<1000 900 700>
    componentLinearAccScale=<0.3 0.3 0.25>
    componentLinearVelScale=<0.3 0.3 0.3>
   }
   n2 = Component To Local(ComponentPose=@n1)
   n3 = Use cached pose 'post not rigid body pose'()
   n4 = Get Is backpack on() -> (Is backpack on:bool)
   Output Pose(Result=Blend Poses by bool(
     BlendPose_0=Blend Poses by bool(BlendPose_0=@n3, BlendPose_1=@n2, bActiveValue=@n4) {blendType=Linear}
     BlendPose_1=Use cached pose 'post not rigid body pose'()
     bActiveValue=Get bRigidBodyEnabled()
    ) {blendType=Linear}) {layerGroup=None}
   n5 = Anim Dynamics - Bone: dyn_bang_01_l(ComponentPose=Local To Component(LocalPose=Copy Pose From Mesh() {bUseAttachedParent=true}), Alpha=1, ExternalForce="0,0,0") {
    angularDampingOverride=0.999
    bOverrideAngularDamping=true
    simulationSpace=World
    alphaBoolBlend {blendOption=Linear}
    boundBone {boneName=dyn_bang_01_l}
    componentAppliedLinearAccClamp=<3000 3000 3000>
    componentLinearAccScale=<0.4 0.8 0.7>
    componentLinearVelScale=<0.4 0.4 0.5>
    physicsBodyDefinitions [{
      boundBone {boneName=dyn_bang_01_l}
      boxExtents=<15 15 15>
      constraintSetup {angularLimitsMax=<0 8 0>, angularLimitsMin=<0 -8 -20>, angularTarget=<0 0 0>}
      localJointOffset=<-0.4 0 0>
     }]
    relativeSpaceBone {boneName=spine_05}
   }
   n6 = Anim Dynamics - Bone: dyn_bang_02_l(ComponentPose=@n5, Alpha=1, ExternalForce="0,0,0") {
    angularDampingOverride=0.999
    bOverrideAngularDamping=true
    simulationSpace=World
    alphaBoolBlend {blendOption=Linear}
    boundBone {boneName=dyn_bang_02_l}
    componentAppliedLinearAccClamp=<3000 3000 3000>
    componentLinearAccScale=<0.4 0.7 0.7>
    componentLinearVelScale=<0.4 0.4 0.5>
    physicsBodyDefinitions [{
      boundBone {boneName=dyn_bang_02_l}
      boxExtents=<15 15 15>
      constraintSetup {angularLimitsMax=<0 8 0>, angularLimitsMin=<0 -8 -10>, angularTarget=<0 0 0>}
      localJointOffset=<-0.4 0 0>
     }]
    relativeSpaceBone {boneName=spine_05}
   }
   n7 = Anim Dynamics - Bone: dyn_bang_03_l(ComponentPose=@n6, Alpha=1, ExternalForce="0,0,0") {
    angularDampingOverride=0.999
    bOverrideAngularDamping=true
    simulationSpace=World
    alphaBoolBlend {blendOption=Linear}
    boundBone {boneName=dyn_bang_03_l}
    componentAppliedLinearAccClamp=<3000 3000 3000>
    componentLinearAccScale=<0.4 0.4 0.7>
    componentLinearVelScale=<0.4 0.4 0.5>
    physicsBodyDefinitions [{
      boundBone {boneName=dyn_bang_03_l}
      boxExtents=<15 15 15>
      constraintSetup {angularLimitsMax=<0 0 0>, angularLimitsMin=<0 -10 -7>, angularTarget=<0 0 0>}
      localJointOffset=<-0.3 0 0>
     }]
    relativeSpaceBone {boneName=spine_05}
   }
   n8 = Anim Dynamics - Bone: dyn_bang_01_r(ComponentPose=@n7, Alpha=1, ExternalForce="0,0,0") {
    angularDampingOverride=0.999
    bOverrideAngularDamping=true
    simulationSpace=World
    alphaBoolBlend {blendOption=Linear}
    boundBone {boneName=dyn_bang_01_r}
    componentAppliedLinearAccClamp=<3000 3000 3000>
    componentLinearAccScale=<0.4 0.7 0.7>
    componentLinearVelScale=<0.4 0.4 0.5>
    physicsBodyDefinitions [{
      boundBone {boneName=dyn_bang_01_r}
      boxExtents=<15 15 15>
      constraintSetup {angularLimitsMax=<0 8 0>, angularLimitsMin=<0 -8 -20>, angularTarget=<0 0 0>}
      localJointOffset=<-0.4 0 0>
     }]
    relativeSpaceBone {boneName=spine_05}
   }
   n9 = Anim Dynamics - Bone: dyn_bang_02_r(ComponentPose=@n8, Alpha=1, ExternalForce="0,0,0") {
    angularDampingOverride=0.999
    bOverrideAngularDamping=true
    simulationSpace=World
    alphaBoolBlend {blendOption=Linear}
    boundBone {boneName=dyn_bang_02_r}
    componentAppliedLinearAccClamp=<3000 3000 3000>
    componentLinearAccScale=<0.4 0.4 0.7>
    componentLinearVelScale=<0.4 0.4 0.5>
    physicsBodyDefinitions [{
      boundBone {boneName=dyn_bang_02_r}
      boxExtents=<15 15 15>
      constraintSetup {angularLimitsMax=<0 15 0>, angularLimitsMin=<0 -5 -6>, angularTarget=<0 0 0>}
      localJointOffset=<-0.3 0 0>
     }]
    relativeSpaceBone {boneName=spine_05}
   }
   Save cached pose 'post bang dyn pose'(Pose=Component To Local(ComponentPose=@n9)) {cacheName=post bang dyn pose}
   n10 = "Transform (Modify) Bone - Bone: dyn_braid_spline_00"(
    ComponentPose=Local To Component(LocalPose=Use cached pose 'post bang dyn pose'())
    Alpha=1
    Translation="-0.000,-1.898,-1.973"
    Rotation="0,0,0"
    Scale="1,1,1"
   ) {
    rotationMode=BMM_Additive
    rotationSpace=BCS_BoneSpace
    translationMode=BMM_Additive
    translationSpace=BCS_BoneSpace
    alphaBoolBlend {blendOption=Linear}
    boneToModify {boneName=dyn_braid_spline_00}
    translation {y=-1.898, z=-1.973}
   }
   n11 = "Transform (Modify) Bone - Bone: dyn_braid_spline_01"(ComponentPose=@n10, Alpha=1, Translation="0.000,-1.031,-2.309", Rotation="0,0,0", Scale="1,1,1") {
    rotationMode=BMM_Additive
    rotationSpace=BCS_BoneSpace
    translationMode=BMM_Additive
    translationSpace=BCS_BoneSpace
    alphaBoolBlend {blendOption=Linear}
    boneToModify {boneName=dyn_braid_spline_01}
    translation {y=-1.031, z=-2.309}
   }
   n12 = Get bIsTargeting()
   n13 = "Transform (Modify) Bone - Bone: dyn_braid_spline_01"!(ComponentPose=@n11, bAlphaBoolEnabled:bool=@n12, Translation="-1,0,0", Rotation="0,0,0", Scale="1,1,1") -> (Pose) {
    alphaInputType=Bool
    rotationMode=BMM_Additive
    rotationSpace=BCS_BoneSpace
    translationMode=BMM_Additive
    translationSpace=BCS_BoneSpace
    alphaBoolBlend {blendInTime=0.4, blendOption=Linear, blendOutTime=0.4}
    boneToModify {boneName=dyn_braid_spline_01}
    translation {x=-1}
   }
   n14 = "Transform (Modify) Bone - Bone: dyn_braid_spline_02"!(ComponentPose=@n13, bAlphaBoolEnabled:bool=@n12, Translation="-1,0,0", Rotation="0,0,0", Scale="1,1,1") -> (Pose) {
    alphaInputType=Bool
    rotationMode=BMM_Additive
    rotationSpace=BCS_BoneSpace
    translationMode=BMM_Additive
    translationSpace=BCS_BoneSpace
    alphaBoolBlend {blendInTime=0.4, blendOption=Linear, blendOutTime=0.4}
    boneToModify {boneName=dyn_braid_spline_02}
    translation {x=-1}
   }
   n15 = Control Rig(Source=Component To Local(ComponentPose=@n14)) {
    alphaBoolBlend {blendOption=Linear}
    controlRigAssetReference {blueprintRigClass=&/F_MED_Crane_Laugh_FaceAcc_CtrlRig.F_MED_Crane_Laugh_FaceAcc_CtrlRig_C}
   }
   Save cached pose 'post cape wearing pose'(Pose=@n15) {cacheName=post cape wearing pose}
   n16 = Anim Dynamics - Bone: dyn_braid_05(ComponentPose=Local To Component(LocalPose=Use cached pose 'post cape wearing pose'()), Alpha=1, ExternalForce="0,0,0") {
    angularDampingOverride=0.9
    bOverrideAngularDamping=true
    simulationSpace=RootRelative
    alphaBoolBlend {blendOption=Linear}
    boundBone {boneName=dyn_braid_05}
    componentAppliedLinearAccClamp=<300 300 300>
    componentLinearAccScale=<0.5 0.5 0.5>
    componentLinearVelScale=<0.5 0.5 0.5>
    physicsBodyDefinitions [{
      boundBone {boneName=dyn_braid_05}
      boxExtents=<15 15 15>
      constraintSetup {angularLimitsMax=<0 0 3>, angularLimitsMin=<0 -3 -3>, angularTarget=<0 0 0>}
      localJointOffset=<0.4 0 0>
     }]
   }
   n17 = Anim Dynamics - Bone: dyn_braid_06(ComponentPose=@n16, Alpha=1, ExternalForce="0,0,0") {
    angularDampingOverride=0.9
    bOverrideAngularDamping=true
    simulationSpace=RootRelative
    alphaBoolBlend {blendOption=Linear}
    boundBone {boneName=dyn_braid_06}
    componentAppliedLinearAccClamp=<300 300 300>
    componentLinearAccScale=<0.5 0.5 0.5>
    componentLinearVelScale=<0.5 0.5 0.5>
    physicsBodyDefinitions [{
      boundBone {boneName=dyn_braid_06}
      boxExtents=<13 13 13>
      constraintSetup {angularLimitsMax=<0 0 4>, angularLimitsMin=<0 -3 -4>, angularTarget=<0 0 0>}
      localJointOffset=<0.4 0 0>
     }]
   }
   n18 = Anim Dynamics - Bone: dyn_braid_07(ComponentPose=@n17, Alpha=1, ExternalForce="0,0,0") {
    angularDampingOverride=0.95
    bOverrideAngularDamping=true
    simulationSpace=RootRelative
    alphaBoolBlend {blendOption=Linear}
    boundBone {boneName=dyn_braid_07}
    componentAppliedLinearAccClamp=<300 300 300>
    componentLinearAccScale=<0.5 0.5 0.7>
    componentLinearVelScale=<0.5 0.5 0.5>
    physicsBodyDefinitions [{
      boundBone {boneName=dyn_braid_07}
      boxExtents=<14 14 14>
      constraintSetup {angularLimitsMax=<0 3 10>, angularLimitsMin=<0 -8 -10>, angularTarget=<0 0 0>}
      localJointOffset=<0.4 0 0>
     }]
   }
   n19 = Anim Dynamics - Bone: dyn_braid_08(ComponentPose=@n18, Alpha=1, ExternalForce="0,0,0") {
    angularDampingOverride=0.95
    bOverrideAngularDamping=true
    simulationSpace=RootRelative
    alphaBoolBlend {blendOption=Linear}
    boundBone {boneName=dyn_braid_08}
    componentAppliedLinearAccClamp=<300 300 300>
    componentLinearAccScale=<0.5 0.5 0.7>
    componentLinearVelScale=<0.5 0.5 0.5>
    physicsBodyDefinitions [{
      boundBone {boneName=dyn_braid_08}
      boxExtents=<11 11 11>
      constraintSetup {angularLimitsMax=<0 4 15>, angularLimitsMin=<0 -12 -15>, angularTarget=<0 0 0>}
      localJointOffset=<0.4 0 0>
     }]
   }
   n20 = Get bIsSkyDiving()
   n21 = "Transform (Modify) Bone - Bone: dyn_braid_07"!(ComponentPose=@n19, bAlphaBoolEnabled:bool=@n20, Translation="0,0,0", Rotation="10,0,0", Scale="1,1,1") -> (Pose) {
    alphaInputType=Bool
    rotationMode=BMM_Additive
    rotationSpace=BCS_BoneSpace
    translationMode=BMM_Additive
    translationSpace=BCS_BoneSpace
    alphaBoolBlend {blendInTime=0.4, blendOption=Linear, blendOutTime=0.4}
    boneToModify {boneName=dyn_braid_07}
    rotation {pitch=10}
   }
   n22 = "Transform (Modify) Bone - Bone: dyn_braid_08"!(ComponentPose=@n21, bAlphaBoolEnabled:bool=@n20, Translation="0,0,0", Rotation="15,0,0", Scale="1,1,1") -> (Pose) {
    alphaInputType=Bool
    rotationMode=BMM_Additive
    rotationSpace=BCS_BoneSpace
    translationMode=BMM_Additive
    translationSpace=BCS_BoneSpace
    alphaBoolBlend {blendInTime=0.4, blendOption=Linear, blendOutTime=0.4}
    boneToModify {boneName=dyn_braid_08}
    rotation {pitch=15}
   }
   Save cached pose 'post not rigid body pose'(Pose=Blend Poses by bool(
     BlendPose_0=Use cached pose 'post cape wearing pose'()
     BlendPose_1=Component To Local(ComponentPose=@n22)
     bActiveValue=Get Is cape on() -> (Is cape on:bool)
    ) {blendType=Linear}) {cacheName=post not rigid body pose}
  ]}
 EventGraph {nodes [
   Event BlueprintUpdateAnimation()
   > n1 = Gravity Override State Function DEPRECATED(
    CharacterPartAnimInstance=Self-Reference()
    GameTimeInSeconds=GetGameTimeInSeconds()
    GravityOverrideParams=Get Gravity Override Params()
    RBAN Params=Get RBAN Params()
   )
   > Set Gravity Override(Gravity Override=@n1)
   n2 = CheckForBackbling()
   > Set Is cape on(Is cape on:bool=@n2.bIsBackblingACape) -> (Output_Get:bool)
   > Set Is backpack on(Is backpack on:bool=OR Boolean(A=@n2, B=@n2.bIsBackblingLong)) -> (Output_Get:bool)
  ]}
]
variables [
 Gravity Override {type=Vector}
 Gravity Override Params {
  type=GravityOverrideParamsStruct
  default {
   skydive_FloatingDown {gravityOverride=<0 -700 -400>}
   skydive_GoingUp {gravityOverride=<0 -300 -700>}
   skydive_Parachuting {gravityOverride=<0 -300 -980>}
  }
 }
 Is backpack on {type=boolean}
 Is cape on {type=boolean}
 RBAN Params {
  type=RBANParamsStruct
  default {
   dBNO {simSpaceSettings {
     maxAngularAcceleration=1000
     maxAngularVelocity=3
     maxLinearAcceleration=300
     maxLinearVelocity=700
     velocityScaleZ=0
     externalLinearDragV=<0 0.5 0>
    }}
   dBNO_Carried {simSpaceSettings {
     maxAngularAcceleration=1000
     maxAngularVelocity=3
     maxLinearAcceleration=300
     maxLinearVelocity=700
     velocityScaleZ=0
     externalLinearDragV=<0 0.5 0>
    }}
   emoteOrMelee {simSpaceSettings {
     maxAngularAcceleration=1000
     maxAngularVelocity=3
     maxLinearAcceleration=300
     maxLinearVelocity=700
     velocityScaleZ=0
     externalLinearDragV=<0 0.5 0>
    }}
   front_End {simSpaceSettings {
     maxAngularAcceleration=1000
     maxAngularVelocity=3
     maxLinearAcceleration=300
     maxLinearVelocity=700
     velocityScaleZ=0
     externalLinearDragV=<0 0.5 0>
    }}
   onGround_Crouch_Moving {simSpaceSettings {
     maxAngularAcceleration=1000
     maxAngularVelocity=3
     maxLinearAcceleration=300
     maxLinearVelocity=700
     velocityScaleZ=0
     externalLinearDragV=<0 0.5 0>
    }}
   onGround_Falling {simSpaceSettings {
     maxAngularAcceleration=1000
     maxAngularVelocity=3
     maxLinearAcceleration=300
     maxLinearVelocity=700
     velocityScaleZ=0
     externalLinearDragV=<0 0.5 0>
    }}
   onGround_Moving {simSpaceSettings {
     maxAngularAcceleration=1000
     maxAngularVelocity=3
     maxLinearAcceleration=300
     maxLinearVelocity=700
     velocityScaleZ=0
     externalLinearDragV=<0 0.5 0>
    }}
   onGround_Standing {simSpaceSettings {
     maxAngularAcceleration=1000
     maxAngularVelocity=3
     maxLinearAcceleration=300
     maxLinearVelocity=700
     velocityScaleZ=0
     externalLinearDragV=<0 0.5 0>
    }}
   rail_Grinding {simSpaceSettings {
     maxAngularAcceleration=1000
     maxAngularVelocity=3
     maxLinearAcceleration=300
     maxLinearVelocity=700
     velocityScaleZ=0
     worldAlpha=0.1
     externalLinearDragV=<0 0.5 0>
    }}
   riding_Moving {simSpaceSettings {
     maxAngularAcceleration=1000
     maxAngularVelocity=3
     maxLinearAcceleration=300
     maxLinearVelocity=700
     velocityScaleZ=0
     externalLinearDragV=<0 0.5 0>
    }}
   riding_Moving_Targeting {simSpaceSettings {
     maxAngularAcceleration=1000
     maxAngularVelocity=3
     maxLinearAcceleration=300
     maxLinearVelocity=700
     velocityScaleZ=0
     externalLinearDragV=<0 0.5 0>
    }}
   riding_Standing {simSpaceSettings {
     maxAngularAcceleration=1000
     maxAngularVelocity=3
     maxLinearAcceleration=300
     maxLinearVelocity=700
     velocityScaleZ=0
     externalLinearDragV=<0 0.5 0>
    }}
   skydive_DivingDown {simSpaceSettings {
     maxAngularAcceleration=1000
     maxAngularVelocity=3
     maxLinearAcceleration=300
     maxLinearVelocity=700
     velocityScaleZ=0
     externalLinearDragV=<0 0.5 0>
    }}
   skydive_FloatingDown {
    componentLinearAccScale=<0 0.5 0>
    componentLinearVelScale=<0 0.5 0>
    simSpaceSettings {
     maxAngularAcceleration=1000
     maxAngularVelocity=3
     maxLinearAcceleration=300
     maxLinearVelocity=700
     velocityScaleZ=0
     externalLinearDragV=<0 0.5 0>
    }
   }
   skydive_GoingUp {simSpaceSettings {
     maxAngularAcceleration=1000
     maxAngularVelocity=3
     maxLinearAcceleration=300
     maxLinearVelocity=700
     velocityScaleZ=0
     externalLinearDragV=<0 0.5 0>
    }}
   skydive_Parachuting {
    componentLinearAccScale=<0 0.1 0.3>
    componentLinearVelScale=<0 0.1 0.3>
    simSpaceSettings {
     maxAngularAcceleration=1000
     maxAngularVelocity=3
     maxLinearAcceleration=300
     maxLinearVelocity=700
     velocityScaleZ=0
     externalLinearDragV=<0 0.5 0>
    }
   }
   skydive_Surfing {simSpaceSettings {
     maxAngularAcceleration=1000
     maxAngularVelocity=3
     maxLinearAcceleration=300
     maxLinearVelocity=700
     velocityScaleZ=0
     externalLinearDragV=<0 0.5 0>
    }}
   swimming {simSpaceSettings {
     maxAngularAcceleration=1000
     maxAngularVelocity=3
     maxLinearAcceleration=300
     maxLinearVelocity=700
     velocityScaleZ=0
     externalLinearDragV=<0 0.5 0>
    }}
   tactical_Sprint {simSpaceSettings {
     maxAngularAcceleration=1000
     maxAngularVelocity=3
     maxLinearAcceleration=300
     maxLinearVelocity=700
     velocityScaleZ=0
     externalLinearDragV=<0 0.5 0>
    }}
   ziplining {simSpaceSettings {
     maxAngularAcceleration=1000
     maxAngularVelocity=3
     maxLinearAcceleration=300
     maxLinearVelocity=700
     velocityScaleZ=0
     externalLinearDragV=<0 0.5 0>
    }}
  }
 }
]
